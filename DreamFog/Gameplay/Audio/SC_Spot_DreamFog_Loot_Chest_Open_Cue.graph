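# Sound cue for opening a DreamFog loot chest, layering DreamFog chest-open sounds over stock treasure chest sweeteners.
# A quality-level node sends its first two slots to a mix of two four-way random layers, with an envelope briefly suppressing the sweetener's onset.
# The last two slots use a lighter mix of one fixed sweetener and two random opens.
# Modulators vary pitch, and concurrency allows one instance per owner, blocking new ones.
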SoundCue /DreamFog/Gameplay/Audio/SC_Spot_DreamFog_Loot_Chest_Open_Cue
root=QualityLevel_0
nodes [
 {
  id=Enveloper_1
  children [Random_6]
  properties {
   pitchCurve {editorCurveData {keys [{time=0}]}}
   volumeCurve {editorCurveData {keys [{interpMode=RCIM_Cubic, tangentMode=RCTM_Break, time=0.1, value=0.01}, {interpMode=RCIM_Cubic, tangentMode=RCTM_Break, time=0.12, value=1}]}}
  }
 }
 {id=Mixer_0, children [Modulator_1,Modulator_3], properties {inputVolume [1,3]}}
 {id=Mixer_1, children [Modulator_0,Modulator_2], properties {inputVolume [1,1.5]}}
 {id=Modulator_0, children [WavePlayer_1], properties {pitchMax=1.5, pitchMin=1, volumeMax=0.7, volumeMin=0.7}}
 {id=Modulator_1, children [Enveloper_1], properties {pitchMax=1.2, pitchMin=1, volumeMax=0.7, volumeMin=0.7}}
 {id=Modulator_2, children [Random_2], properties {pitchMin=1}}
 {id=Modulator_3, children [Random_1], properties {pitchMin=1, volumeMax=1, volumeMin=1}}
 {id=QualityLevel_0, children [Mixer_0,Mixer_0,Mixer_1,Mixer_1]}
 {id=Random_1, children [WavePlayer_7, WavePlayer_8, WavePlayer_9, WavePlayer_10], properties {weights [1,1,1,1]}}
 {id=Random_2, children [WavePlayer_15, WavePlayer_20], properties {weights [1,1]}}
 {id=Random_6, children [WavePlayer_16, WavePlayer_17, WavePlayer_18, WavePlayer_19], properties {weights [1,1,1,1]}}
 {id=WavePlayer_1, properties {soundWaveAssetPtr=/Game/Sounds/Foley_Loot/Containers/Treasure_Chest/treasure_chest_sweet_shorter_02_opt_02}}
 {id=WavePlayer_10, properties {soundWaveAssetPtr=&/Spot_DreamFog_Loot_Chest_Open_01}}
 {id=WavePlayer_15, properties {soundWaveAssetPtr=&/Spot_DreamFog_Loot_Chest_Open_02}}
 {id=WavePlayer_16, properties {soundWaveAssetPtr=/Game/Sounds/Foley_Loot/Containers/Treasure_Chest/treasure_chest_sweet_shorter_01}}
 {id=WavePlayer_17, properties {soundWaveAssetPtr=/Game/Sounds/Foley_Loot/Containers/Treasure_Chest/treasure_chest_sweet_shorter_02}}
 {id=WavePlayer_18, properties {soundWaveAssetPtr=/Game/Sounds/Foley_Loot/Containers/Treasure_Chest/treasure_chest_sweet_shorter_03}}
 {id=WavePlayer_19, properties {soundWaveAssetPtr=/Game/Sounds/Foley_Loot/Containers/Treasure_Chest/treasure_chest_sweet_shorter_04}}
 {id=WavePlayer_20, properties {soundWaveAssetPtr=&Spot_DreamFog_Loot_Chest_Open_01}}
 {id=WavePlayer_7, properties {soundWaveAssetPtr=&/Spot_DreamFog_Loot_Chest_Open_04}}
 {id=WavePlayer_8, properties {soundWaveAssetPtr=&/Spot_DreamFog_Loot_Chest_Open_03}}
 {id=WavePlayer_9, properties {soundWaveAssetPtr=&Spot_DreamFog_Loot_Chest_Open_02}}
]
properties {
 attenuationSettings=/Game/Sounds/Attenuation/Fort_Item_Interact_Att
 bOverrideConcurrency=true
 maxDistance=1850
 soundClassObject=&Fort_Loot_Reward_PSM
 volumeMultiplier=1
 concurrencyOverrides {bLimitToOwner=true, maxCount=1, resolutionRule=PreventNew}
}
