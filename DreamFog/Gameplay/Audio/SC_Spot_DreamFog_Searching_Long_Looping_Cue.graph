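# Looping sound cue for a long Dream Fog search interaction, built from Infest LootSac interaction sounds.
# A mixer layers three branches: a random one-shot start, a random looping bed, and a looping Shepard-tone layer.
# Both looping layers pass through envelopers with short volume fade-ins, and each branch gets pitch modulation.
# It uses mono item-interact attenuation and limits playback to one instance per owner, preventing new ones.
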SoundCue /DreamFog/Gameplay/Audio/SC_Spot_DreamFog_Searching_Long_Looping_Cue
root=Mixer_0
nodes [
 {
  id=Enveloper_0
  children [Modulator_1]
  properties {pitchCurve {editorCurveData {keys [{time=0}]}}, volumeCurve {editorCurveData {keys [<at 0 0 RCIM_Cubic>,<at 0.25 1>]}}}
 }
 {
  id=Enveloper_2
  children [Modulator_0]
  properties {pitchCurve {editorCurveData {keys [{time=0}]}}, volumeCurve {editorCurveData {keys [<at 0 0 RCIM_Cubic>,<at 0.25 1>]}}}
 }
 {id=Mixer_0, children [Modulator_2,Enveloper_2,Enveloper_0], properties {inputVolume [1.7,2,1.2]}}
 {id=Modulator_0, children [Random_2], properties {pitchMax=1.1, volumeMax=1, volumeMin=1}}
 {id=Modulator_1, children [WavePlayer_2], properties {pitchMax=1.1, volumeMax=1, volumeMin=1}}
 {id=Modulator_2, children [Random_1], properties {pitchMax=1.2, pitchMin=1, volumeMax=1, volumeMin=1}}
 {id=Random_1, children [WavePlayer_0,WavePlayer_1,WavePlayer_3], properties {bShouldExcludeFromBranchCulling=true, weights [1,1,1]}}
 {id=Random_2, children [WavePlayer_5, WavePlayer_9, WavePlayer_10], properties {bShouldExcludeFromBranchCulling=true, weights [1,1,1]}}
 {id=WavePlayer_0, properties {soundWaveAssetPtr=/BugBiomeGameplay/Gameplay/Rewards/Audio/Infest_LootSac_Interact_Start_01}}
 {id=WavePlayer_1, properties {soundWaveAssetPtr=/BugBiomeGameplay/Gameplay/Rewards/Audio/Infest_LootSac_Interact_Start_02}}
 {id=WavePlayer_10, properties {bLooping=true, soundWaveAssetPtr=/BugBiomeGameplay/Gameplay/Rewards/Audio/Infest_LootSac_Interact_Loop_03}}
 {
  id=WavePlayer_2
  properties {bLooping=true, soundWaveAssetPtr=/BugBiomeGameplay/Gameplay/Rewards/Audio/Infest_LootSac_Interact_ShepardTone_Loop_01}
 }
 {id=WavePlayer_3, properties {soundWaveAssetPtr=/BugBiomeGameplay/Gameplay/Rewards/Audio/Infest_LootSac_Interact_Start_03}}
 {id=WavePlayer_5, properties {bLooping=true, soundWaveAssetPtr=/BugBiomeGameplay/Gameplay/Rewards/Audio/Infest_LootSac_Interact_Loop_01}}
 {id=WavePlayer_9, properties {bLooping=true, soundWaveAssetPtr=/BugBiomeGameplay/Gameplay/Rewards/Audio/Infest_LootSac_Interact_Loop_02}}
]
properties {
 attenuationSettings=/Game/Sounds/Attenuation/Fort_Item_Interact_Mono_Att
 bOverrideConcurrency=true
 maxDistance=1500
 soundClassObject=&Fort_Loot_Collect
 volumeMultiplier=0.8183
 concurrencyOverrides {bLimitToOwner=true, maxCount=1, resolutionRule=PreventNew}
}
