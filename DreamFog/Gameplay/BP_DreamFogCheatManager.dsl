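; DreamFog cheat manager that spawns haunted containers and haunted POI takeovers for the controlled pawn, only where cosmetic events cannot execute (dedicated server).
; Containers spawn ahead of the pawn; POI takeovers spawn at the pawn or once at a stored location, with surrounding containers.
; Existing takeovers are cleaned up and destroyed before replacement, with a timer-reset cooldown gating that cleanup in the at-pawn version.
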(fn MakeSpawnLocationsForPlayerLocation
  (params (PlayerPawn :type Pawn%20Object%20Reference))
  (return (Utilities|Array|MakeArray (+ (Transformation|GetActorLocation PlayerPawn) (* (Transformation|GetActorForwardVector PlayerPawn) 1.5e+03)) (+ (Transformation|GetActorLocation PlayerPawn) (* (Transformation|GetActorRightVector PlayerPawn) 1.5e+03)) (Utilities|Operators|Add (Transformation|GetActorLocation PlayerPawn) (* (Transformation|GetActorRightVector PlayerPawn) 6e+02) "250.000000,250.000000,0.000000") (Utilities|Operators|Add (Transformation|GetActorLocation PlayerPawn) (* (Transformation|GetActorForwardVector PlayerPawn) 1.5e+03) "500.000000,500.000000,0.000000"))))

(fn ResetAbleToSpawnNewHaunting
  (Variables|Default|SetbSpawnCooldown false))

(fn SpawnHauntedContainers
  (bind _returnvalue (CheatManager|GetPlayerController))
  (bind _returnvalue_2 (Pawn|GetControlledPawn _returnvalue))
  (Networking|CanExecuteCosmeticEvents
    (:False
      (Utilities|IsValid _returnvalue_2
        (:Is%20Valid
          (Game|SpawnActorfromClass "/DreamFog/Gameplay/B_DreamFog_Container.B_DreamFog_Container_C" (Math|Transform|MakeTransform (+ (- (Transformation|GetActorLocation _returnvalue_2) "0.000000,0.000000,83.000000") (* (Transformation|GetActorForwardVector _returnvalue_2) 5e+02))) "AlwaysSpawn" "MultiplyWithRoot" 0 0 0 0))))))

(fn SpawnHauntedPOI
  (bind _returnvalue (CheatManager|GetPlayerController))
  (bind _returnvalue_2 (Pawn|GetControlledPawn _returnvalue))
  (Networking|CanExecuteCosmeticEvents
    (:False
      (Utilities|IsValid _returnvalue_2
        (:Is%20Valid
          (bind _returnvalue_3 (Actor|GetActorOfClass "/DreamFog/Gameplay/B_HauntedPOI_Takeover.B_HauntedPOI_Takeover_C"))
          (Utilities|IsValid _returnvalue_3
            (:Is%20Valid
              (if (not (Variables|Default|GetbSpawnCooldown))
                (Class|BHauntedPOITakeover|CleanupSpawnedActors _returnvalue_3)
                (Actor|DestroyActor _returnvalue_3)))
            (:Is%20Not%20Valid
              (Utilities|FlowControl|Sequence
                (:then_0
                  _)
                (:then_1
                  (for _array_element (CallFunction|MakeSpawnLocationsforPlayerLocation _returnvalue_2)
                    (Game|SpawnActorfromClass "/DreamFog/Gameplay/B_DreamFog_Container.B_DreamFog_Container_C" (Math|Transform|MakeTransform _array_element) "AlwaysSpawn" "MultiplyWithRoot" 0 0 0 0))
                  (Variables|Default|SetbSpawnCooldown true)
                  (Utilities|Time|SetTimerbyEvent (EventDispatchers|CreateEvent :!event ResetAbleToSpawnNewHaunting) 6.0)))))
          (bind _returnvalue_4 (Game|SpawnActorfromClass "/DreamFog/Gameplay/B_HauntedPOI_Takeover.B_HauntedPOI_Takeover_C" (Math|Transform|MakeTransform (Transformation|GetActorLocation _returnvalue_2)) "AlwaysSpawn" "MultiplyWithRoot" 0 0))
          (Class|BHauntedPOITakeover|TakeoverCheatedIn _returnvalue_4))))))

(fn SpawnHauntedPOIAtLocation
  (bind _returnvalue (CheatManager|GetPlayerController))
  (bind _returnvalue_2 (Pawn|GetControlledPawn _returnvalue))
  (bind _spawn_transform_location (Variables|Default|GetSpawn%20Transform%20Location))
  (Networking|CanExecuteCosmeticEvents
    (:False
      (Utilities|IsValid _returnvalue_2
        (:Is%20Valid
          (if (not (Variables|Default|GetbSpawnedAtLocation))
            (Variables|Default|SetbSpawnedAtLocation true)
            (bind _returnvalue_3 (Actor|GetActorOfClass "/DreamFog/Gameplay/B_HauntedPOI_Takeover.B_HauntedPOI_Takeover_C"))
            (Utilities|IsValid _returnvalue_3
              (:Is%20Valid
                (Class|BHauntedPOITakeover|CleanupSpawnedActors _returnvalue_3)
                (Actor|DestroyActor _returnvalue_3))
              (:Is%20Not%20Valid
                _))
            (bind _returnvalue_4 (Game|SpawnActorfromClass "/DreamFog/Gameplay/B_HauntedPOI_Takeover.B_HauntedPOI_Takeover_C" (Math|Transform|MakeTransform _spawn_transform_location) "AlwaysSpawn" "MultiplyWithRoot" 0 0))
            (Class|BHauntedPOITakeover|TakeoverCheatedIn _returnvalue_4)
            (Game|SpawnActorfromClass "/DreamFog/Gameplay/B_DreamFog_Container.B_DreamFog_Container_C" (Math|Transform|MakeTransform (- (+ _spawn_transform_location (* (Transformation|GetActorForwardVector _returnvalue_2) 1.5e+03)) "0.000000,0.000000,83.000000")) "AlwaysSpawn" "MultiplyWithRoot" 0 0 0 0)
            (for _array_element (Variables|Default|GetContainerSpawnsAtLocation)
              (Game|SpawnActorfromClass "/DreamFog/Gameplay/B_DreamFog_Container.B_DreamFog_Container_C" (Math|Transform|MakeTransform _array_element) "AlwaysSpawn" "MultiplyWithRoot" 0 0 0 0))))))))
