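; DreamFog pumpkin chest loot container Blueprint that handles the chest's visual change when looted.
; On loot, it deactivates the idle Niagara effect and destroys the pumpkin chest mesh component.
; Where cosmetic events can execute, it spawns an enlarged opening Niagara burst at the static mesh component's location.
; BeginPlay only calls the parent, and the ContainerOpened event holds only a delegate bind to itself.
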(event EventBeginPlay
  (|CallParentFunction))

(event Custom|ContainerOpened_Event
  (Default|BindEventtoContainerOpened (delegate Custom|ContainerOpened_Event)))

(event EventOnLoot
  (Components|Activation|Deactivate (Variables|Default|GetNS_Idle))
  (Components|DestroyComponent (Variables|Default|GetSM_Hermes_AxisThrow_PumpkinChest_A_B))
  (Networking|CanExecuteCosmeticEvents
    (:True
      (Niagara|SpawnSystematLocation :SystemTemplate "/DreamFog/Gameplay/FX/NS_PumpkinChest_Open.NS_PumpkinChest_Open" :Location (Transformation|GetWorldLocation (Variables|Debug|GetStaticMeshComponent)) :Scale "2.000000,2.000000,2.000000" :bAutoDestroy true :bAutoActivate true :PoolingMethod "None" :bPreCullCheck true))))

(fn ConstructionScript)
