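; Haunted POI manager that decides on the server, once the playlist is ready, which points of interest become haunted.
; A scalable-float toggle selects data-layer mode, which shuffles enabled setups, picks a configured number, then activates and unloads each one's layers.
; Otherwise, after a short delay, it picks random POI tags, triggers matching takeover actors and destroys the rest.
; A cheat event spawns and triggers a takeover at a random player.
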(event EventBeginPlay
  (|CallParentFunction)
  (if (Networking|IsServer)
    (Class|FortAsyncActionCurrentPlaylistReady|PlaylistDataReadyAsync
      (:OnReady
        (if (ScalableFloat|GetValueasBoolwithContext (Variables|Default|GetHF_UseDataLayerLogic))
          (bind _enableddatalayers (Default|GetAllEnabledDataLayers (Variables|Default|GetDataLayerList)))
          (bind _output_get (Variables|Default|SetEnabledLayers _enableddatalayers))
          (if (and (> (Utilities|Array|Length _output_get) 0) (> (ScalableFloat|GetValueasInteger (Variables|Default|GetHF_NumToEnable)) 0))
            (bind _enableddatalayers_2 (Default|SelectDataLayerstoEnable (Variables|Default|GetEnabledLayers)))
            (for _array_element _enableddatalayers_2
              (Default|ApplyDataLayerSetup _array_element)))
          (else
            (Utilities|FlowControl|Delay 3.0)
            (CallFunction|SelectPOI)))))))

(event Custom|SpawnHauntedPOI
  ;[1 0.2632 0.6864] Cheat ke* SpawnHauntedPOI. Spawn in Haunted POI through a cheat.
  (bind _outfortplayerpawns (Fortnite|GetFortPlayerPawns))
  (bind _returnvalue (Game|SpawnActorfromClass (Variables|Default|GetTakeoverClass) (Math|Transform|MakeTransform (Transformation|GetActorLocation (Utilities|Array|RandomArrayItem _outfortplayerpawns))) "Undefined" "MultiplyWithRoot" 0 0))
  (Class|BHauntedPOITakeover|TakeoverSelected _returnvalue))

(fn ApplyDataLayerSetup
  (params (RandomDataLayerData :type Random%20Data%20Layer%20Data%20Structure%20%28by%20ref%29))
  (bind (_enableddatalayers_9_41bb9a3e4648805b5af70aa728288c92 _disableddatalayers_10_8de59e3b474c88ed2e7988b185572444) (Utilities|Struct|BreakRandomDataLayerData RandomDataLayerData))
  (bind _returnvalue (WorldPartition|GetDataLayerManager))
  (Utilities|FlowControl|Sequence
    (:then_0
      (for _array_element _enableddatalayers_9_41bb9a3e4648805b5af70aa728288c92
        (DataLayers|SetDataLayerInstanceRuntimeState _returnvalue (DataLayers|GetDataLayerInstancefromAsset _returnvalue _array_element) "Activated" false)))
    (:then_1
      (for _array_element_2 _disableddatalayers_10_8de59e3b474c88ed2e7988b185572444
        (DataLayers|SetDataLayerInstanceRuntimeState _returnvalue (DataLayers|GetDataLayerInstancefromAsset _returnvalue _array_element_2) "Unloaded" false)))))

(fn ChoosePOI_Tags
  (LocalVariables|SetRemainingTags (Variables|Default|GetPOITags))
  (Utilities|FlowControl|ForLoopwithBreak 0
    (:LoopBody
      (if (Utilities|Array|IsNotEmpty (LocalVariables|GetRemainingTags))
        _
        (else
          _))
      (bind _output_get (LocalVariables|SetCurrentTag (Utilities|Array|RandomArrayItem (LocalVariables|GetRemainingTags))))
      (Development|PrintString (Utilities|String|Append (GameplayTags|GetDebugStringfromGameplayTag _output_get) " chosen to be dark") true true "(R=0.000000,G=0.660000,B=1.000000,A=1.000000)" 1e+01)
      (Utilities|Array|Add (LocalVariables|GetLocalOutTags) _output_get)
      (Utilities|Array|RemoveItem (LocalVariables|GetRemainingTags) _output_get))
    (:Completed
      _))
  (return (LocalVariables|GetLocalOutTags)))

(fn GetAllEnabledDataLayers
  (params (DataLayersToCheck :type Array%20of%20Random%20Data%20Layer%20Data%20Structures))
  (for _array_element DataLayersToCheck
    (if (ScalableFloat|GetValueasBool (.bSetupEnabled_22_436E2FE243B567AE5E7CA9AAC2346CA0 (Utilities|Struct|BreakRandomDataLayerData _array_element)))
      (Utilities|Array|Add (LocalVariables|GetFilteredLayers) _array_element)))
  (return (LocalVariables|GetFilteredLayers)))

(fn GetTakeover
  (return (Variables|Default|GetTakeoverActors)))

(fn SelectDataLayersToEnable
  (params (DataLayersToCheck :type Array%20of%20Random%20Data%20Layer%20Data%20Structures))
  (bind _output_get (LocalVariables|SetLayersToSelect DataLayersToCheck))
  (Utilities|Array|Shuffle _output_get)
  (for _index (range (+ (Math|Integer|SelectInt (Utilities|Array|Length DataLayersToCheck) (- (ScalableFloat|GetValueasInteger (Variables|Default|GetHF_NumToEnable)) 1) (< (Utilities|Array|Length DataLayersToCheck) (- (ScalableFloat|GetValueasInteger (Variables|Default|GetHF_NumToEnable)) 1))) 1))
    (bind _result (ArrayMacros|SafeArrayGet (LocalVariables|GetLayersToSelect) _index)
      (:%20%20
        (Utilities|Array|Add (LocalVariables|GetFilteredLayers) _result))))
  (return (LocalVariables|GetFilteredLayers)))

(fn SelectPOI
  (bind _outtags (CallFunction|ChoosePOITags))
  (bind _output_get (LocalVariables|SetSelectedPOITags _outtags))
  (if (Utilities|Array|IsNotEmpty _output_get)
    (bind _takeoveractorarray (CallFunction|GetTakeover))
    (for _array_element _takeoveractorarray
      (if (Utilities|Array|ContainsItem (LocalVariables|GetSelectedPOITags) (Variables|/DreamFog/Gameplay/B_HauntedPOI_Takeover.B_HauntedPOI_Takeover_C|GetPOITag _array_element))
        (Class|BHauntedPOITakeover|TakeoverSelected _array_element)
        (else
          (Actor|DestroyActor _array_element))))
    (Development|PrintString "Completed Actor Array loop" true true "(R=0.000000,G=0.660000,B=1.000000,A=1.000000)" 15.0)))

(fn ConstructionScript)
