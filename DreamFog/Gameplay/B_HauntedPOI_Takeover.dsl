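; Haunted POI takeover actor that grants player pawns inside its volume a dash gameplay effect and looping cue, removing both on exit unless hiding-prop rules apply.
; An optional brush volume replaces the default overlap component once playlist data is ready; overlaps are handled only where cosmetic events cannot run.
; Default-volume overlaps start a delayed retry for cached players missing the effect, covering respawns.
; Takeover events toggle bSpawnActors, whose OnRep does no work.
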(event EventEndPlay
  (params (EndPlayReason :type EEndPlayReason%20Enum :default "Destroyed"))
  ;[1 0.4406 0 0.5] End Play
  (|CallParentFunction EndPlayReason)
  (bind _overlapbrushvolume (Variables|Default|GetOverlapBrushVolume)
    (:then
      (Collision|UnbindEventfromOnActorBeginOverlap _overlapbrushvolume (EventDispatchers|CreateEvent :!event Custom|BrushVolumeOverlapped))
      (Collision|UnbindEventfromOnActorEndOverlap _overlapbrushvolume (EventDispatchers|CreateEvent :!event Custom|BrushOverlapEnded)))
    (:else)))

(event EventBeginPlay
  ;[0 1 0.1131 0.5] Begin Play
  (|CallParentFunction)
  (Class|FortAsyncActionCurrentPlaylistReady|PlaylistDataReadyAsync
    (:OnReady
      (bind _overlapbrushvolume (Variables|Default|GetOverlapBrushVolume)
        (:then
          (Collision|BindEventtoOnActorBeginOverlap _overlapbrushvolume (EventDispatchers|CreateEvent :!event Custom|BrushVolumeOverlapped))
          (Collision|BindEventtoOnActorEndOverlap _overlapbrushvolume (EventDispatchers|CreateEvent :!event Custom|BrushOverlapEnded))
          (bind _overlapvolume (Variables|Default|GetOverlapVolume)
            (:then
              (Components|DestroyComponent _overlapvolume))
            (:else)))
        (:else)) ; If a brush volume wants to be used instead of a basic shape volume.
    )))

(event Custom|BrushVolumeOverlapped
  (params (OverlappedActor :type Actor%20Object%20Reference) (OtherActor :type Actor%20Object%20Reference))
  (Networking|CanExecuteCosmeticEvents
    (:False
      (CallFunction|ActorOverlappedVolume OtherActor))))

(event Custom|CheckForPlayerEffect
  ; Apply GE to player if they didn't get it applied when initially overlapping. Fixes case of respawning in the volume and not getting the effect applied
  (for _array_element (Variables|Default|GetOverlappedPlayers)
    (bind _returnvalue (GameplayTags|HasMatchingGameplayTag(Message) _array_element "(TagName=\"Gameplay.Status.SunRose.SilkyWolf.Active\")"))
    (if (not _returnvalue)
      (CallFunction|ActorOverlappedVolume _array_element)))
  (Utilities|Array|Clear (Variables|Default|GetOverlappedPlayers)))

(event OnComponentBeginOverlap(OverlapVolume)
  (params (OverlappedComponent :type Primitive%20Component%20Object%20Reference) (OtherActor :type Actor%20Object%20Reference) (OtherComp :type Primitive%20Component%20Object%20Reference) (OtherBodyIndex :type Integer) (bFromSweep :type Boolean) (SweepResult :type Hit%20Result%20Structure%20%28by%20ref%29))
  (Networking|CanExecuteCosmeticEvents
    (:False
      (CallFunction|ActorOverlappedVolume OtherActor)
      (Utilities|Time|SetTimerbyEvent (EventDispatchers|CreateEvent :!event Custom|CheckForPlayerEffect) (ScalableFloat|GetValueatLevelwithContext (Variables|Default|GetHF_DashRespawnDelay))))))

(event Custom|BrushOverlapEnded
  (params (OverlappedActor :type Actor%20Object%20Reference) (OtherActor :type Actor%20Object%20Reference))
  (Networking|CanExecuteCosmeticEvents
    (:False
      (CallFunction|ActorOverlapEnded OtherActor))))

(event OnComponentEndOverlap(OverlapVolume)
  (params (OverlappedComponent :type Primitive%20Component%20Object%20Reference) (OtherActor :type Actor%20Object%20Reference) (OtherComp :type Primitive%20Component%20Object%20Reference) (OtherBodyIndex :type Integer))
  (Networking|CanExecuteCosmeticEvents
    (:False
      (CallFunction|ActorOverlapEnded OtherActor))))

(event Custom|TakeoverCheatedIn
  (Variables|Default|SetbSpawnActors true)
  (Networking|FlushNetDormancy))

(event Custom|CleanupSpawnedActors
  ;[0 0 0] Cheats
  (Networking|CanExecuteCosmeticEvents
    (:False
      (Variables|Default|SetbSpawnActors false))))

(event Custom|TakeoverSelected
  (Variables|Default|SetbSpawnActors true)
  (Networking|FlushNetDormancy))

(event Custom|DisableDash
  (params (bShouldBeEnabled :type Boolean))
  ;[0 0 0] Cheats
  (Variables|Default|SetbDashingDisabled bShouldBeEnabled))

(fn ActorOverlapEnded
  (params (OverlappedActor :type Actor%20Object%20Reference))
  (Utilities|IsA(soft) OverlappedActor "/Script/FortniteGame.FortPlayerPawn"
    (:True
      (bind _returnvalue (GameplayTags|HasMatchingGameplayTag(Message) OverlappedActor "(TagName=\"Pawn.Athena.NonParticipant\")"))
      (if (not (or _returnvalue (CallFunction|CheckforHidingProp OverlappedActor)))
        (Utilities|FlowControl|Sequence
          (:then_0
            (if (not (or (Variables|Default|GetbDashingDisabled) (not (ScalableFloat|GetValueasBoolwithContext (Variables|Default|GetHF_DashingEnabled)))))
              (GameplayEffects|RemoveActiveGameplayEffectbySourceEffect (Ability|GetAbilitySystemComponent OverlappedActor) (Variables|Default|GetGE_ToApply) 0 -1)))
          (:then_1
            (if (ScalableFloat|GetValueasBoolwithContext (Variables|Default|GetHF_CueEnabled))
              (GameplayCue|RemoveGameplayCueOnActor(Looping) OverlappedActor (Variables|Default|GetGameplay%20Cue%20Tag) (Variables|Default|GetParameters)))))))))

(fn ActorOverlappedVolume
  (params (OverlappingActor :type Actor%20Object%20Reference))
  (Utilities|IsA(soft) OverlappingActor "/Script/FortniteGame.FortPlayerPawn"
    (:True
      (bind _returnvalue (GameplayTags|HasMatchingGameplayTag(Message) OverlappingActor "(TagName=\"Pawn.Athena.NonParticipant\")"))
      (if (not _returnvalue)
        (Utilities|FlowControl|Sequence
          (:then_0
            (if (not (or (Variables|Default|GetbDashingDisabled) (not (ScalableFloat|GetValueasBoolwithContext (Variables|Default|GetHF_DashingEnabled)))))
              (Environment|ApplyGlobalEnvironmentGameplayEffecttoActor OverlappingActor (Variables|Default|GetGE_ToApply))
              (bind _returnvalue_2 (GameplayTags|HasMatchingGameplayTag(Message) OverlappingActor "(TagName=\"Gameplay.Status.SunRose.SilkyWolf.Active\")")) ; If player still doesn't have the GE applied cache them for later.
              (if (not _returnvalue_2)
                (Utilities|Array|AddUnique (Variables|Default|GetOverlappedPlayers) (Variables|GetOverlappingActor)))))
          (:then_1
            (if (ScalableFloat|GetValueasBoolwithContext (Variables|Default|GetHF_CueEnabled))
              (GameplayCue|AddGameplayCueOnActor(Looping) (Variables|GetOverlappingActor) (Variables|Default|GetGameplay%20Cue%20Tag) (Variables|Default|GetParameters)))))))))

(fn CheckForHidingProp
  (params (Actor :type Actor%20Object%20Reference))
  (bind _returnvalue (GameplayTags|HasAnyMatchingGameplayTags(Message) Actor (Variables|Default|GetTagsToIgnore)))
  (bind _returnvalue_2 (GameplayTags|HasMatchingGameplayTag(Message) Actor "(TagName=\"Granted.Athena.EnvItem.HidingProp.Teleporter.Teleporting\")"))
  (return (Math|Boolean|ANDBoolean (not _returnvalue_2) _returnvalue (ScalableFloat|GetValueasBoolwithContext (Variables|Default|GetHF_GainChargeWhileHiding)))))

(fn OnRep_bSpawnActors
  (if (Variables|Default|GetbSpawnActors)
    (Networking|CanExecuteCosmeticEvents)
    (else
      (Networking|CanExecuteCosmeticEvents))))

(fn ConstructionScript)
