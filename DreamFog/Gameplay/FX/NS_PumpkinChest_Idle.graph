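# Idle Niagara effect for a pumpkin chest, built around a Sparkles emitter of warm orange, short-lived sprites.
# Sparkles spawn continuously in an offset sphere, drift under gravity, curl noise and drag, and fade and shrink via life curves.
# Three Rays emitters are listed without modules or renderers.
# The system exposes intensity, scale, offset, hue and pulse parameters and references a time-of-day scalar input, with no wiring shown.
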NiagaraSystem /DreamFog/Gameplay/FX/NS_PumpkinChest_Idle
dependencies {
 usedDataInterfaces [&NiagaraDataInterfaceCurve]
 usedDynamicInputs [
  &Multiply_Float
  &MakeLinearColorFromVectorAndFloat
  &FNTimeOfDayScalar
 ]
 usedModules [
  &self:SystemSpawnScript.SystemScriptSource.SystemScriptGraph.NiagaraNodeAssignment_2.SetVariables_#1
  &SystemState
  &self:SystemSpawnScript.SystemScriptSource.SystemScriptGraph.NiagaraNodeAssignment_0.SetVariables_#2
 ]
}
emitters {
 Rays {rendererClasses [], renderers []}
 Rays001 {rendererClasses [], renderers []}
 Rays002 {rendererClasses [], renderers []}
 Sparkles {
  emitterUpdateScript [
   &EmitterState()
   &SpawnRate(SpawnRate=24, Use Spawn Probability=false)
  ]
  particleSpawnScript [
   /Niagara/Modules/Spawn/Initialization/InitializeParticle(
    Color=<rgba 2 0.8 0.2 1>
    Lifetime=&UniformRangedFloat(Minimum=0.2, Maximum=1.3)
    Module.Write Color=true
    Module.Write Lifetime=true
    Module.Write Mass=true
    Module.Write Material Random=true
    Module.Write Position=false
    Module.Write Scale=true
    Module.Write SpriteRotation=false
    Module.Write SpriteSize=true
    Position=&SimulationPosition()
    Sprite Size=&Vector2DFromFloat(Value=&UniformRangedFloat(Minimum=5, Maximum=20))
   )
   &ShapeLocation(
    MaskSpawnGroup=false
    Offset=<-5 0 50>
    Offset Mode=::NewEnumerator0
    Override Seed=false
    Sphere Radius=40
    Sphere Surface Distribution=0.8
    UseOwnerScale=true
   )
   &AddVelocity(Coordinate Space=::Simulation, Velocity=<0 25 0>)
   &ApplyInitialForces()
  ]
  particleUpdateScript [
   &ParticleState()
   &GravityForce(Gravity=<0 0 -200>)
   /Niagara/Modules/Update/Forces/CurlNoiseForce(Noise Frequency=0.2, Noise Strength=150, Pan Noise=false) {module=CurlNoiseForce001}
   &Drag(Drag=3, Use Linear Drag=true, Use Rotational Drag=false)
   &SolveForcesAndVelocity(Clamp Velocity=false, Limit Acceleration=false)
   &ScaleColor(
    Scale Alpha=&FloatFromCurve(FloatCurve=DataInterface() {propertyValues {Curve {keys [<at -0.0092 0.0144>,<at 0.0967 0.9767>,<at 1.0011 0.002>]}}})
    Scale RGB=&VectorFromFloat(Value=2)
    ScaleA=true
    ScaleRGB=true
   )
   &ScaleSpriteSize(Scale Factor=&Vector2DFromFloat(Value=&FloatFromCurve(FloatCurve=DataInterface() {propertyValues {Curve {keys [<at -0.0046 0>,<at 0.1746 0.9952>,<at 1.0011 0.002>]}}})))
  ]
  rendererClasses [&NiagaraSpriteRendererProperties]
  renderers [{rendererClass=&NiagaraSpriteRendererProperties}]
 }
}
summary {
 systemName=NS_PumpkinChest_Idle
 emitters [
  {emitterName=Sparkles, rendererClasses [&NiagaraSpriteRendererProperties]}
  {emitterName=Rays, rendererClasses []}
  {emitterName=Rays001, rendererClasses []}
  {emitterName=Rays002, rendererClasses []}
 ]
 userVariables [
  User.FX_Intensity {type=NiagaraFloat, defaultValue=0.75, description=""}
  User.FXScale {type=NiagaraFloat, defaultValue=0.6, description=""}
  User.GlobalOffset {type=Vector3f, description="", defaultValue {struct=Vector3f, value=<0 0 0>}}
  User.HueShift {type=NiagaraFloat, defaultValue=1, description=""}
  User.Pulse_Speed {type=NiagaraFloat, defaultValue=0.1, description=""}
 ]
}
